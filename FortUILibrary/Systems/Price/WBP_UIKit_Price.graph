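# Base price display widget for the UI kit, with no layout of its own, showing V-Bucks or real-money prices, regular prices and disclaimers.
# Setters write text and visibility through stub getter functions and optional text-block variables, switching the V-Buck and platform store icons by currency type.
# Real-price and legal text come from FortMtxStoreVM only for Mtx currency when the store is ready.
# Construct and Destruct start and stop that view model's update monitoring.
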WidgetBlueprint /FortUILibrary/Systems/Price/WBP_UIKit_Price
graphs [
 EventGraph {nodes [
   Event Construct()
   n1 = Get FortMtxStoreVM() -> (FortMtxStoreVM:Fort Mtx Store VM Object Reference)
   > StartMonitoringForUpdates!(execute, self:Fort Mcp Catalog VM Object Reference=@n1, RequestingObject:Object Reference) -> (then)
   Event Destruct()
   > StopMonitoringForUpdates!(execute, self:Fort Mcp Catalog VM Object Reference=@n1, RequestingObject:Object Reference) -> (then)
  ]}
 GetAdditionalPriceTextRef {nodes [
   GetAdditionalPriceTextRef() {is=title K2Node_FunctionEntry}
   > Return Node!(execute, Widget:Common Text Object Reference)
  ]}
 GetAdditionalRegularPriceTextRef_Optional {nodes [
   GetAdditionalRegularPriceTextRef_Optional() {is=title K2Node_FunctionEntry}
   > Return Node()
  ]}
 GetPlatformIconRef {nodes [
   GetPlatformIconRef() {is=title K2Node_FunctionEntry}
   > Return Node(Widget:WBP Price Platform Store Icon Object Reference)
  ]}
 GetPriceTextRef {nodes [
   GetPriceTextRef() {is=title K2Node_FunctionEntry}
   > Return Node(Widget:Common Text Object Reference)
  ]}
 GetRealPriceLegalText {nodes [
   n1 = GetRealPriceLegalText!() -> (then, MtxPrice:int, CurrencyType:EFortPriceCurrency Enum=Mtx, MtxStoreLoadState:EFortMtxStoreLoadingState Enum=Uninitialized) {is=title K2Node_FunctionEntry}
   n2 = Get RealPriceLegalText_Optional() -> (RealPriceLegalText_Optional:Common Text Object Reference)
   > n3 = Branch(Condition=AND Boolean!(
     A=IsValid(Object=@n2)
     B="Equal (Enum)"(A:EFortPriceCurrency Enum=@n1.CurrencyType, B:EFortPriceCurrency Enum=Mtx)
     C:bool="Equal (Enum)"(A:EFortMtxStoreLoadingState Enum=@n1.MtxStoreLoadState, B:EFortMtxStoreLoadingState Enum=Ready)
    ) -> (ReturnValue))
   > Return Node!(
    execute
    ReturnValue:Text=GetRealPriceLegalText!(
     self:Fort Mtx Store VM Object Reference=Get FortMtxStoreVM()
     MtxPrice:int=Get MtxPrice() -> (MtxPrice:int=0)
     bIsAdditionalText:bool=Get ShowDoubleAsterisk() -> (ShowDoubleAsterisk:bool)
    ) -> (ReturnValue:Text) {is=title K2Node_CallFunction}
   )
   Return Node(execute=@n3.else)
  ]}
 GetRealPriceText {nodes [
   n1 = GetRealPriceText!() -> (then, MtxPrice:int, CurrencyType:EFortPriceCurrency Enum=Mtx, MtxStoreLoadState:EFortMtxStoreLoadingState Enum=Uninitialized) {is=title K2Node_FunctionEntry}
   n2 = Get RealPriceText_Optional() -> (RealPriceText_Optional:Common Text Object Reference)
   > n3 = Branch(Condition=AND Boolean(
     A=IsValid(Object=@n2)
     B="Equal (Enum)"(A:EFortPriceCurrency Enum=@n1.CurrencyType, B:EFortPriceCurrency Enum=Mtx)
     C="Equal (Enum)"(A:EFortMtxStoreLoadingState Enum=@n1.MtxStoreLoadState, B:EFortMtxStoreLoadingState Enum=Ready)
    ))
   > Return Node(ReturnValue=GetRealPriceText!(
     self:Fort Mtx Store VM Object Reference=Get FortMtxStoreVM()
     MtxPrice:int=Get MtxPrice() -> (MtxPrice=0)
     bIsAdditionalText:bool=Get ShowDoubleAsterisk()
    ) -> (ReturnValue:Text) {is=title K2Node_CallFunction})
   Return Node(execute=@n3.else)
  ]}
 GetRegularPriceContainerRef_Optional {nodes [
   GetRegularPriceContainerRef_Optional() {is=title K2Node_FunctionEntry}
   > Return Node!(execute, Widget:Widget Object Reference)
  ]}
 GetRegularPriceTextRef_Optional {nodes [
   GetRegularPriceTextRef_Optional() {is=title K2Node_FunctionEntry}
   > Return Node(Widget:Common Text Object Reference)
  ]}
 GetVBuckIconRef {nodes [
   GetVBuckIconRef() {is=title K2Node_FunctionEntry}
   > Return Node(Widget:Widget Object Reference)
  ]}
 SetAdditionalPriceText {nodes [
   n1 = SetAdditionalPriceText!() -> (then, AdditionalPriceText:Text) {is=title K2Node_FunctionEntry}
   > Set Text(Text=@n1, self=GetAdditionalPriceTextRef!(self) -> (Widget:Common Text Object Reference) {is=title K2Node_CallFunction})
  ]}
 SetAdditionalRegularPriceText {nodes [
   n1 = SetAdditionalRegularPriceText!() -> (then, AdditionalRegularPriceText:Text) {is=title K2Node_FunctionEntry}
   n2 = GetAdditionalRegularPriceTextRef_Optional!(self) -> (Widget:Common Text Object Reference) {is=title K2Node_CallFunction}
   > Is Valid(InputObject=@n2)
   > Set Text(Text=@n1, self=@n2)
  ]}
 SetDiscountDisclaimerText {nodes [
   n1 = SetDiscountDisclaimerText!() -> (then, DiscountDisclaimerText:Text) {is=title K2Node_FunctionEntry}
   > n2 = Get!(execute:Exec) -> (then:Exec, else:Exec, DiscountDisclaimerText_Optional:Common Text Object Reference)
   > Set Text(Text=@n1, self=@n2)
  ]}
 SetDiscountDisclaimerVisibility {nodes [
   n1 = SetDiscountDisclaimerVisibility!() -> (then, Visibility:ESlateVisibility Enum=Visible) {is=title K2Node_FunctionEntry}
   > n2 = Get()
   > SetVisibility(self=@n2, InVisibility=@n1)
  ]}
 SetPlatformIconVisibility {nodes [
   n1 = SetPlatformIconVisibility!() -> (then, CurrencyType:EFortPriceCurrency Enum=Mtx) {is=title K2Node_FunctionEntry}
   > UpdateCurrency!(
    execute
    self:WBP Price Platform Store Icon Object Reference=GetPlatformIconRef!(self) -> (Widget:WBP Price Platform Store Icon Object Reference) {is=title K2Node_CallFunction}
    CurrencyType:EFortPriceCurrency Enum=@n1
   ) -> (then)
  ]}
 SetPrice {nodes [
   n1 = SetPrice!() -> (then, Price:Text) {is=title K2Node_FunctionEntry}
   > Set Text(Text=@n1, self=GetPriceTextRef!(self) -> (Widget:Common Text Object Reference) {is=title K2Node_CallFunction})
  ]}
 SetRealPriceLegalText {nodes [
   n1 = SetRealPriceLegalText!() -> (then, RealPriceLegalText:Text) {is=title K2Node_FunctionEntry}
   > n2 = Get!(execute) -> (then, else, RealPriceLegalText_Optional:Common Text Object Reference)
   > Set Text(Text=@n1, self=@n2)
  ]}
 SetRealPriceText {nodes [
   n1 = SetRealPriceText!() -> (then, RealPriceText:Text) {is=title K2Node_FunctionEntry}
   > n2 = Get!(execute) -> (then, else, RealPriceText_Optional:Common Text Object Reference)
   > Set Text(Text=@n1, self=@n2)
  ]}
 SetRegularPrice {nodes [
   SetRegularPrice!() -> (then, RegularPrice:Text) {is=title K2Node_FunctionEntry}
   > n1 = Sequence()
   n2 = GetRegularPriceTextRef_Optional!(self) -> (Widget:Common Text Object Reference) {is=title K2Node_CallFunction}
   > Is Valid(InputObject=@n2)
   > Set Text(Text=Get RegularPrice() -> (RegularPrice:Text), self=@n2)
   n3 = GetRegularPriceContainerRef_Optional!(self) -> (Widget:Widget Object Reference) {is=title K2Node_CallFunction}
   Is Valid(exec=@n1.then_1, InputObject=@n3)
   n4 = Get RegularPrice()
   > Set Visibility(Visibility=Select(Option 0=SelfHitTestInvisible, Option 1=Collapsed, Index=TextIsEmpty(InText=@n4)), self=@n3)
  ]}
 SetVbuckIconVisibility {nodes [
   n1 = SetVbuckIconVisibility!() -> (then, CurrencyType:EFortPriceCurrency Enum=Mtx) {is=title K2Node_FunctionEntry}
   > Set Visibility(
    Visibility=Select!(Mtx:ESlateVisibility Enum=SelfHitTestInvisible, RealMoney:ESlateVisibility Enum=Collapsed, Index:EFortPriceCurrency Enum=@n1) -> (ReturnValue)
    self=GetVBuckIconRef!(self) -> (Widget:Widget Object Reference) {is=title K2Node_CallFunction}
   )
  ]}
]
variables [
 DiscountDisclaimerText_Optional {type=CommonTextBlock, default=null}
 RealPriceLegalText_Optional {type=CommonTextBlock, default=null}
 RealPriceText_Optional {type=CommonTextBlock, default=null}
 ShowDoubleAsterisk {type=boolean}
]
widgets {info {inheritedWidgetCount=0, parentClass=&UserWidget, rootWidgetClass=null, widgetCount=0}, widgets []}
